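# Sound cue for Zeus boss alert-idle voice chatter, drawn from the Season 29 VO waves.
# A root Random node picks among five equally weighted wave players, each holding one distant-variant AlertIdle line, and is excluded from branch culling.
# Playback uses the Zeus boss close attenuation, the critical enemy VO sound class, a boosted volume and a Zeus chatter concurrency set to limit overlapping lines.
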SoundCue /AIGods_Zeus/Audio/VO/Cues/VO_Zeus_S29_2023_AlertIdle_Cue
root=Random_1
nodes [
 {
  id=Random_1
  children [WavePlayer_5,WavePlayer_6,WavePlayer_7,WavePlayer_8,WavePlayer_9]
  properties {bShouldExcludeFromBranchCulling=true, weights [1,1,1,1,1]}
 }
 {id=WavePlayer_5, properties {soundWaveAssetPtr=/AIGods_Zeus/Audio/VO/Waves/Dist/VO_Zeus_S29_2023_AlertIdle_010}}
 {id=WavePlayer_6, properties {soundWaveAssetPtr=/AIGods_Zeus/Audio/VO/Waves/Dist/VO_Zeus_S29_2023_AlertIdle_021}}
 {id=WavePlayer_7, properties {soundWaveAssetPtr=/AIGods_Zeus/Audio/VO/Waves/Dist/VO_Zeus_S29_2023_AlertIdle_030}}
 {id=WavePlayer_8, properties {soundWaveAssetPtr=/AIGods_Zeus/Audio/VO/Waves/Dist/VO_Zeus_S29_2023_AlertIdle_040}}
 {id=WavePlayer_9, properties {soundWaveAssetPtr=/AIGods_Zeus/Audio/VO/Waves/Dist/VO_Zeus_S29_2023_AlertIdle_050}}
]
properties {
 attenuationSettings=/AIGods_Zeus/Audio/VO/Fort_S29_Zeus_Boss_Close_Att
 maxDistance=2097152
 soundClassObject=&Fort_VO_Enemy_Critical
 volumeMultiplier=4.2699
 concurrencySet [/AIGods_Zeus/Audio/VO/Zeus_Chatter_Conc]
}
