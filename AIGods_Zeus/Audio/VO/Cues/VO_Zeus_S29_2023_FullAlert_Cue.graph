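# Sound cue for Zeus boss voice lines, named for the full-alert state.
# A Random node picks with equal weight among four waves: two distant full-alert takes and two Phase 2 transform takes.
# Playback uses boosted volume, the critical enemy VO sound class and the boss close attenuation, with the Zeus chatter concurrency set governing overlap.
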SoundCue /AIGods_Zeus/Audio/VO/Cues/VO_Zeus_S29_2023_FullAlert_Cue
root=Random_1
nodes [
 {id=Random_1, children [WavePlayer_10, WavePlayer_6, WavePlayer_11, WavePlayer_9], properties {weights [1,1,1,1]}}
 {id=WavePlayer_10, properties {soundWaveAssetPtr=/AIGods_Zeus/Audio/VO/Waves/VO_Zeus_S29_2023_Phase2_Transform_011}}
 {id=WavePlayer_11, properties {soundWaveAssetPtr=/AIGods_Zeus/Audio/VO/Waves/VO_Zeus_S29_2023_Phase2_Transform_012}}
 {id=WavePlayer_6, properties {soundWaveAssetPtr=/AIGods_Zeus/Audio/VO/Waves/VO_Zeus_S29_2023_FullAlert_Dist_021}}
 {id=WavePlayer_9, properties {soundWaveAssetPtr=/AIGods_Zeus/Audio/VO/Waves/VO_Zeus_S29_2023_FullAlert_Dist_050}}
]
properties {
 attenuationSettings=/AIGods_Zeus/Audio/VO/Fort_S29_Zeus_Boss_Close_Att
 maxDistance=2097152
 soundClassObject=&Fort_VO_Enemy_Critical
 volumeMultiplier=4.2376
 concurrencySet [/AIGods_Zeus/Audio/VO/Zeus_Chatter_Conc]
}
