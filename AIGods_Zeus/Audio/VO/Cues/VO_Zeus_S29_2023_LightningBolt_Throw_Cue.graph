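# Sound cue for Zeus boss voice lines played when throwing a lightning bolt.
# A random node picks one of four equally weighted distant-variant wave players.
# Playback uses the enemy critical VO sound class, a Zeus boss close attenuation asset, a chatter concurrency group and a boosted volume multiplier.
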SoundCue /AIGods_Zeus/Audio/VO/Cues/VO_Zeus_S29_2023_LightningBolt_Throw_Cue
root=Random_1
nodes [
 {id=Random_1, children [WavePlayer_4,WavePlayer_5,WavePlayer_6,WavePlayer_7], properties {weights [1,1,1,1]}}
 {id=WavePlayer_4, properties {soundWaveAssetPtr=/AIGods_Zeus/Audio/VO/Waves/Dist/VO_Zeus_S29_2023_LightningBolt_Throw_010}}
 {id=WavePlayer_5, properties {soundWaveAssetPtr=/AIGods_Zeus/Audio/VO/Waves/Dist/VO_Zeus_S29_2023_LightningBolt_Throw_011}}
 {id=WavePlayer_6, properties {soundWaveAssetPtr=/AIGods_Zeus/Audio/VO/Waves/Dist/VO_Zeus_S29_2023_LightningBolt_Throw_012}}
 {id=WavePlayer_7, properties {soundWaveAssetPtr=/AIGods_Zeus/Audio/VO/Waves/Dist/VO_Zeus_S29_2023_LightningBolt_Throw_013}}
]
properties {
 attenuationSettings=/AIGods_Zeus/Audio/VO/Fort_S29_Zeus_Boss_Close_Att
 maxDistance=2097152
 soundClassObject=&Fort_VO_Enemy_Critical
 volumeMultiplier=6.0313
 concurrencySet [/AIGods_Zeus/Audio/VO/Zeus_Chatter_Conc]
}
