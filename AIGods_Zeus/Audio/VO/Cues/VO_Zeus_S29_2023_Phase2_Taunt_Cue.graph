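# Sound cue for the Zeus boss's Phase 2 taunt voice lines from the S29 AI Gods content.
# A Random node picks with equal weight among three taunt waves (010–012) from the Dist folder.
# Playback uses a close boss attenuation, the critical enemy VO sound class, a shared Zeus chatter concurrency group, and a strongly boosted volume.
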SoundCue /AIGods_Zeus/Audio/VO/Cues/VO_Zeus_S29_2023_Phase2_Taunt_Cue
root=Random_1
nodes [
 {id=Random_1, children [WavePlayer_3,WavePlayer_4,WavePlayer_5], properties {weights [1,1,1]}}
 {id=WavePlayer_3, properties {soundWaveAssetPtr=/AIGods_Zeus/Audio/VO/Waves/Dist/VO_Zeus_S29_2023_Phase2_Taunt_010}}
 {id=WavePlayer_4, properties {soundWaveAssetPtr=/AIGods_Zeus/Audio/VO/Waves/Dist/VO_Zeus_S29_2023_Phase2_Taunt_011}}
 {id=WavePlayer_5, properties {soundWaveAssetPtr=/AIGods_Zeus/Audio/VO/Waves/Dist/VO_Zeus_S29_2023_Phase2_Taunt_012}}
]
properties {
 attenuationSettings=/AIGods_Zeus/Audio/VO/Fort_S29_Zeus_Boss_Close_Att
 maxDistance=2097152
 soundClassObject=&Fort_VO_Enemy_Critical
 volumeMultiplier=6.0313
 concurrencySet [/AIGods_Zeus/Audio/VO/Zeus_Chatter_Conc]
}
