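# Sound cue for Zeus boss spawn voice lines, played through the critical enemy VO sound class.
# A root Random node picks one of five spawn wave variants with equal weight.
# A boss-spawn attenuation asset, boosted volume and raised max distance shape its reach, while a Zeus chatter concurrency set limits overlapping lines.
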SoundCue /AIGods_Zeus/Audio/VO/Cues/VO_Zeus_S29_2023_Spawn_Cue
root=Random_0
nodes [
 {id=Random_0, children [WavePlayer_0,WavePlayer_1,WavePlayer_2,WavePlayer_3,WavePlayer_4], properties {weights [1,1,1,1,1]}}
 {id=WavePlayer_0, properties {soundWaveAssetPtr=/AIGods_Zeus/Audio/VO/Waves/VO_Zeus_S29_2023_Spawn_v2_012}}
 {id=WavePlayer_1, properties {soundWaveAssetPtr=/AIGods_Zeus/Audio/VO/Waves/VO_Zeus_S29_2023_Spawn_v2_021}}
 {id=WavePlayer_2, properties {soundWaveAssetPtr=/AIGods_Zeus/Audio/VO/Waves/VO_Zeus_S29_2023_Spawn_v2_030}}
 {id=WavePlayer_3, properties {soundWaveAssetPtr=/AIGods_Zeus/Audio/VO/Waves/VO_Zeus_S29_2023_Spawn_v2_040}}
 {id=WavePlayer_4, properties {soundWaveAssetPtr=/AIGods_Zeus/Audio/VO/Waves/VO_Zeus_S29_2023_Spawn_v2_050}}
]
properties {
 attenuationSettings=/AIGods_Zeus/Audio/VO/Fort_S29_Zeus_Boss_Spawn_Close_Att
 maxDistance=2097152
 priority=2
 soundClassObject=&Fort_VO_Enemy_Critical
 volumeMultiplier=2
 concurrencySet [/AIGods_Zeus/Audio/VO/Zeus_Chatter_Conc]
}
