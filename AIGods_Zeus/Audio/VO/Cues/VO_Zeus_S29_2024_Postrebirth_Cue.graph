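# Sound cue for Zeus boss voice lines played after a rebirth, part of the Season 29 AI Gods audio.
# A Random node picks one of four equally weighted post-rebirth wave variants.
# It routes through the critical enemy VO sound class with boosted volume and a set priority, a close boss-spawn attenuation asset, and a shared Zeus chatter concurrency group to govern overlapping lines.
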SoundCue /AIGods_Zeus/Audio/VO/Cues/VO_Zeus_S29_2024_Postrebirth_Cue
root=Random_0
nodes [
 {id=Random_0, children [WavePlayer_0,WavePlayer_1,WavePlayer_2,WavePlayer_3], properties {weights [1,1,1,1]}}
 {id=WavePlayer_0, properties {soundWaveAssetPtr=/AIGods_Zeus/Audio/VO/Waves/VO_Zeus_S29_2024_Postrebirth_010}}
 {id=WavePlayer_1, properties {soundWaveAssetPtr=/AIGods_Zeus/Audio/VO/Waves/VO_Zeus_S29_2024_Postrebirth_011}}
 {id=WavePlayer_2, properties {soundWaveAssetPtr=/AIGods_Zeus/Audio/VO/Waves/VO_Zeus_S29_2024_Postrebirth_013}}
 {id=WavePlayer_3, properties {soundWaveAssetPtr=/AIGods_Zeus/Audio/VO/Waves/VO_Zeus_S29_2024_Postrebirth_014}}
]
properties {
 attenuationSettings=/AIGods_Zeus/Audio/VO/Fort_S29_Zeus_Boss_Spawn_Close_Att
 maxDistance=2097152
 priority=2
 soundClassObject=&Fort_VO_Enemy_Critical
 volumeMultiplier=2
 concurrencySet [/AIGods_Zeus/Audio/VO/Zeus_Chatter_Conc]
}
